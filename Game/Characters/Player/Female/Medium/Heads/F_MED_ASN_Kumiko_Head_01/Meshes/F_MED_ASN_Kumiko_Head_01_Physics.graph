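# Physics asset for the Kumiko head mesh of the medium female player character, covering dynamic hair and the upper body around it.
# Capsule bodies wrap the upper spine, clavicles, neck, head and eight dyn_hair bones.
# Hair bodies constrain directly to the head: six short ones have zero swing limit and slight twist, while longer 07 and 09 swing wider.
# Neck and clavicle links below the head rotate freely, chaining back to spine_05.
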PhysicsAsset /Game/Characters/Player/Female/Medium/Heads/F_MED_ASN_Kumiko_Head_01/Meshes/F_MED_ASN_Kumiko_Head_01_Physics
bodies {
 clavicle_l [{shapeType=Capsule, length=21.5, radius=12, shapeName=clavicle_l_capsule, center=<22.2875 0 9.471>, rotation=<rot 13.0587 -180 -180>}]
 clavicle_r [{
   shapeType=Capsule
   length=21.5
   radius=12
   shapeName=clavicle_l_capsule
   center=<-19.9975 -10.1957 -2.7997>
   rotation=<rot 22.0028 -165.2568 59.9434>
  }]
 dyn_hair_01 [{shapeType=Capsule, length=3.75, radius=1, shapeName=dyn_hair_01_capsule, center=<0 0 -1.3541>}]
 dyn_hair_02 [{shapeType=Capsule, length=3.25, radius=1.5, shapeName=dyn_hair_02_capsule, center=<0 0 -1.3541>}]
 dyn_hair_03 [{shapeType=Capsule, length=3.25, radius=1.5, shapeName=dyn_hair_03_capsule, center=<0 0.6 -1.3>}]
 dyn_hair_04 [{shapeType=Capsule, length=3.75, radius=1, shapeName=dyn_hair_04_capsule, center=<0 0 -1.3541>}]
 dyn_hair_05 [{shapeType=Capsule, length=3, radius=1.75, shapeName=dyn_hair_05_capsule, center=<0 0.6 -1.3>}]
 dyn_hair_06 [{shapeType=Capsule, length=3.75, radius=1, shapeName=dyn_hair_06_capsule, center=<0 0 -1.3541>}]
 dyn_hair_07 [{shapeType=Capsule, length=10.5, radius=1.5, shapeName=dyn_hair_07_capsule, center=<-1.8796 6.2076 -10.4711>, rotation=<rot -20 0 -4.089>}]
 dyn_hair_09 [{shapeType=Capsule, length=9.25, radius=1, shapeName=dyn_hair_09_capsule, center=<3.3 6.2 -11>, rotation=<rot 23.6434 -0.5267 -7.6961>}]
 head [
  {shapeType=Capsule, length=7, radius=7, center=<4.6179 0 0>, rotation=<rot 90 0 0>}
  {shapeType=Capsule, length=14.5, radius=5.5, shapeName=head_capsule, center=<5.2134 0.5962 0>, rotation=<rot 90 180 -148.4244>}
 ]
 neck_01 [{shapeType=Capsule, length=5.25, radius=4.75, shapeName=neck_01_capsule, rotation=<rot 90 0 0>}]
 neck_02 [{shapeType=Capsule, length=5, radius=5, shapeName=neck_02_capsule, rotation=<rot 90 0 0>}]
 spine_05 [{shapeType=Capsule, length=16, radius=10.5, center=<1.1701 0 0>}]
}
constraints [
 {bone1Name=head, bone2Name=neck_02, swing1LimitDegrees=45, swing2LimitDegrees=45, twistLimitDegrees=45}
 {
  bone1Name=clavicle_l
  bone2Name=spine_05
  swing1LimitDegrees=45
  swing1Motion=Free
  swing2LimitDegrees=45
  swing2Motion=Free
  twistLimitDegrees=45
  twistMotion=Free
 }
 {
  bone1Name=neck_01
  bone2Name=spine_05
  swing1LimitDegrees=45
  swing1Motion=Free
  swing2LimitDegrees=45
  swing2Motion=Free
  twistLimitDegrees=45
  twistMotion=Free
 }
 {
  bone1Name=neck_02
  bone2Name=neck_01
  swing1LimitDegrees=45
  swing1Motion=Free
  swing2LimitDegrees=45
  swing2Motion=Free
  twistLimitDegrees=45
  twistMotion=Free
 }
 {bone1Name=dyn_hair_01, bone2Name=head, swing1LimitDegrees=0, swing2LimitDegrees=0, twistLimitDegrees=5}
 {bone1Name=dyn_hair_02, bone2Name=head, swing1LimitDegrees=0, swing2LimitDegrees=0, twistLimitDegrees=5}
 {bone1Name=dyn_hair_03, bone2Name=head, swing1LimitDegrees=0, swing2LimitDegrees=0, twistLimitDegrees=5}
 {bone1Name=dyn_hair_04, bone2Name=head, swing1LimitDegrees=0, swing2LimitDegrees=0, twistLimitDegrees=5}
 {bone1Name=dyn_hair_05, bone2Name=head, swing1LimitDegrees=0, swing2LimitDegrees=0, twistLimitDegrees=5}
 {bone1Name=dyn_hair_06, bone2Name=head, swing1LimitDegrees=0, swing2LimitDegrees=0, twistLimitDegrees=5}
 {bone1Name=dyn_hair_07, bone2Name=head, swing1LimitDegrees=20, swing2LimitDegrees=30, twistLimitDegrees=8}
 {bone1Name=dyn_hair_09, bone2Name=head, swing1LimitDegrees=20, swing2LimitDegrees=30, twistLimitDegrees=8}
 {
  bone1Name=clavicle_r
  bone2Name=spine_05
  swing1LimitDegrees=45
  swing1Motion=Free
  swing2LimitDegrees=45
  swing2Motion=Free
  twistLimitDegrees=45
  twistMotion=Free
 }
]
